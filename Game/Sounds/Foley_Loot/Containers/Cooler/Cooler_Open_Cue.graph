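# Sound cue for the foley of opening a cooler loot container.
# A single wave player plays Cooler_Open_01 through a modulator that randomizes pitch slightly upward while keeping volume constant.
# The cue routes to the loot collect sound class and SFX submix, with boosted volume, close-prop attenuation and a shared basic foley concurrency limit.
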SoundCue /Game/Sounds/Foley_Loot/Containers/Cooler/Cooler_Open_Cue
root=Modulator_2
nodes [
 {id=Modulator_2, children [WavePlayer_0], properties {pitchMax=1.1, pitchMin=1, volumeMax=1, volumeMin=1}}
 {id=WavePlayer_0, properties {soundWaveAssetPtr=&/Cooler_Open_01}}
]
properties {
 attenuationSettings=/Game/Sounds/Attenuation/Fort_Prop_Close_Att
 maxDistance=1800
 soundClassObject=&Fort_Loot_Collect
 soundSubmixObject=&Fortnite_SFX
 volumeMultiplier=1.75
 concurrencySet [/Game/Sounds/Fort_Foley_Sounds/FoleyLibraries/Settings/Foley_Basic_Global_Conc]
}
