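# Sound cue for the cooler container foley, named for cancelling a search of a cooler.
# A Random node picks one of three equally weighted cooler impact waves and feeds a Modulator left at default settings.
# It plays quietly through the prop search attenuation and loot collect sound class.
# A concurrency override caps simultaneous instances, stopping the farthest first.
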SoundCue /Game/Sounds/Foley_Loot/Containers/Cooler/Cooler_Searching_Cancel_Cue
root=Modulator_3
nodes [
 {id=Modulator_3, children [Random_2]}
 {id=Random_2, children [WavePlayer_5,WavePlayer_6,WavePlayer_7], properties {weights [1,1,1]}}
 {id=WavePlayer_5, properties {soundWaveAssetPtr=&/Cooler_Impact02}}
 {id=WavePlayer_6, properties {soundWaveAssetPtr=&/Cooler_Impact03}}
 {id=WavePlayer_7, properties {soundWaveAssetPtr=&/Cooler_Impact04}}
]
properties {
 attenuationSettings=/Game/Sounds/Attenuation/Fort_Prop_Search_Att
 bOverrideConcurrency=true
 maxDistance=1300
 soundClassObject=&Fort_Loot_Collect
 volumeMultiplier=0.2
 concurrencyOverrides {maxCount=4, resolutionRule=StopFarthestThenPreventNew}
}
