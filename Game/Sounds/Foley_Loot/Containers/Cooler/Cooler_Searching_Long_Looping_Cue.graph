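# Looping sound cue for the long searching foley of a cooler loot container.
# A single wave player loops the Cooler_Search_lp wave through a modulator at the root.
# It uses close-prop attenuation, a boosted volume, the loot-collect sound class, the Fortnite SFX submix and a shared basic foley concurrency setting.
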SoundCue /Game/Sounds/Foley_Loot/Containers/Cooler/Cooler_Searching_Long_Looping_Cue
root=Modulator_0
nodes [
 {id=Modulator_0, children [WavePlayer_0]}
 {id=WavePlayer_0, properties {bLooping=true, soundWaveAssetPtr=&/Cooler_Search_lp}}
]
properties {
 attenuationSettings=/Game/Sounds/Attenuation/Fort_Prop_Close_Att
 maxDistance=1800
 soundClassObject=&Fort_Loot_Collect
 soundSubmixObject=&Fortnite_SFX
 volumeMultiplier=2.8251
 concurrencySet [/Game/Sounds/Fort_Foley_Sounds/FoleyLibraries/Settings/Foley_Basic_Global_Conc]
}
